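# Conversation database for the Pitstop NPC, giving the character's racing-themed greeting and routing into shared NPC dialogue.
# The entry point leads to a single speech node with a default line and contextual lines chosen by priority curves for time since last match, first-ever conversation, or held bars.
# That speech then links out to the shared NPC Library quests and basic services conversation entries.
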ConversationDatabase /NPCLibrary/NPCs/Pitstop/Pitstop/NPCLibrary_Conversation_Pitstop
entry_points [{tag=Conversation.Entry.Pitstop, destinations [#1]}]
nodes [
 ConversationEntryPointNode_7 {
  id=#1
  class=ConversationEntryPointNode
  outputs [#2]
  properties {entryTag {tagName=Conversation.Entry.Pitstop}}
 }
 Speech_C_4 {
  id=#2
  class=Speech_C
  outputs [#3, #4]
  properties {
   nodeName=""
   generalConfig {
    defaultMessage="Keep your foot on the gas. "
    contextualMessages [
     {
      message=Don't worry. We all make our way to the finish line on our own time.
      contextRequirements [{requirement {class=FortControllerRequirement_CheckTimeSinceLastPlayedMatch}}]
      priority {curve {rowName=Tandem.Speech.Priority.WhileSincePlayed}}
     }
     {
      message="In the face of danger, you gotta hit the gas and shift into high gear."
      contextRequirements [{requirement {class=FortControllerRequirement_IsFirstEverConversationWithNPC}}]
      priority {curve {rowName=Tandem.Speech.Priority.FirstEverInteraction}}
     }
     {
      message=That's a lotta gas money.
      contextRequirements [{requirement {class=FortControllerRequirement_HasBars}}]
      priority {curve {rowName=Tandem.Speech.Priority.HighTotalBars}}
     }
    ]
   }
  }
 }
 ConversationLinkNode_0 {id=#4, class=ConversationLinkNode, properties {remoteEntryTag {tagName=Conversation.Entry.BasicServices}}}
 ConversationLinkNode_1 {id=#3, class=ConversationLinkNode, properties {remoteEntryTag {tagName=Conversation.Entry.NPCLibrary.Quests}}}
]
